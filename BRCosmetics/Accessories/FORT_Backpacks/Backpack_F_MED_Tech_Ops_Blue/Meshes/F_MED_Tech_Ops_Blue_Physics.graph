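# Physics asset for the Tech Ops Blue backpack cosmetic, defining simplified collision and joints for its dangling parts.
# The pack, antenna and left and right bullet bones each get a single box body.
# The antenna and both bullet pieces are constrained directly to the pack with limited swing and twist, the antenna held tighter than the bullets.
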PhysicsAsset /BRCosmetics/Accessories/FORT_Backpacks/Backpack_F_MED_Tech_Ops_Blue/Meshes/F_MED_Tech_Ops_Blue_Physics
bodies {
 dyn_antenna [{shapeType=Box, extentX=1.01, extentY=5.4729, extentZ=1.01, center=<-2.4736 -0.0002 0>, rotation=<rot 0 -90 0>}]
 dyn_bullets_l [{shapeType=Box, extentX=6.7027, extentY=3.3757, extentZ=9.1801, center=<0.8991 0.4869 -0.2528>, rotation=<rot -24.9104 -90 -6.1302>}]
 dyn_bullets_r [{shapeType=Box, extentX=6.7027, extentY=3.3757, extentZ=9.1801, center=<1.3865 -0.4869 -0.2528>, rotation=<rot 24.9104 -90 -6.1302>}]
 pack [{shapeType=Box, extentX=22.1176, extentY=9.7592, extentZ=23.1942, center=<0 -5.4896 -4.673>}]
}
constraints [
 {bone1Name=dyn_bullets_l, bone2Name=pack, swing1LimitDegrees=45, swing2LimitDegrees=45, twistLimitDegrees=10}
 {bone1Name=dyn_bullets_r, bone2Name=pack, swing1LimitDegrees=45, swing2LimitDegrees=45, twistLimitDegrees=10}
 {bone1Name=dyn_antenna, bone2Name=pack, swing1LimitDegrees=25, swing2LimitDegrees=25, twistLimitDegrees=5}
]
